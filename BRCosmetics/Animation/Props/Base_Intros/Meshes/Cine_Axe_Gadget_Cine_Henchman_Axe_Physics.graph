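# Physics asset for an axe gadget prop skeletal mesh in the BRCosmetics base intro animations, named for a cinematic henchman axe.
# It holds a single body on the Anim bone, made of one offset, rotated capsule.
# It defines no constraints, so there are no jointed parts to simulate as a ragdoll or chain.
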PhysicsAsset /BRCosmetics/Animation/Props/Base_Intros/Meshes/Cine_Axe_Gadget_Cine_Henchman_Axe_Physics
bodies {Anim [{shapeType=Capsule, length=34.9076, radius=11.2094, center=<0.3537 3.3333 12.5924>, rotation=<rot 4.2941 -90 0.1137>}]}
constraints []
