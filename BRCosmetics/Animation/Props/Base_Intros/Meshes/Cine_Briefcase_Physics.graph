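# Physics asset for Cine_Briefcase, a cinematic briefcase prop used in the base intro animations.
# It defines a single body on the Anim bone, shaped as one capsule offset and rotated to cover the mesh.
# There are no constraints, so the prop is simulated or collides as one rigid piece.
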PhysicsAsset /BRCosmetics/Animation/Props/Base_Intros/Meshes/Cine_Briefcase_Physics
bodies {Anim [{shapeType=Capsule, length=34.0374, radius=27.2987, center=<-1.1808 -1.1546 10.0032>, rotation=<rot 3.4721 -166.7469 -90>}]}
constraints []
