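# Physics asset for Cine_Chair, a chair prop in the Base_Intros cinematic animation set.
# It defines a single body on the Anim bone, approximated by one large, offset and rotated capsule.
# It has no constraints, so the chair is one simple collision volume with no jointed parts.
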PhysicsAsset /BRCosmetics/Animation/Props/Base_Intros/Meshes/Cine_Chair_Physics
bodies {Anim [{shapeType=Capsule, length=61.8366, radius=41.7838, center=<-8.469 1.1167 52.3824>, rotation=<rot -2.6955 -90 -15.7299>}]}
constraints []
