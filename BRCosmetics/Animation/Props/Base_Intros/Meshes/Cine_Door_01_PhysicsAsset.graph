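# Physics asset for the Cine_Door_01 cinematic door prop mesh used in base intro animations.
# It holds a single capsule body on the Anim bone, offset and rotated to roughly enclose the door.
# No constraints are defined, so the asset supplies one simple collision shape rather than an articulated ragdoll.
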PhysicsAsset /BRCosmetics/Animation/Props/Base_Intros/Meshes/Cine_Door_01_PhysicsAsset
bodies {Anim [{shapeType=Capsule, length=133.3954, radius=107.3397, center=<5.1092 -67.5935 -2.8519>, rotation=<rot 19.6061 -90 6.5005>}]}
constraints []
