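# Physics asset for the Cine_Hammer mesh, a prop under the Base_Intros cosmetic animation folder.
# It defines a single capsule body on the Anim bone, offset and rotated to lie along the hammer's shape.
# With no constraints and only one body, it gives the prop simple collision and bounds rather than any jointed simulation.
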PhysicsAsset /BRCosmetics/Animation/Props/Base_Intros/Meshes/Cine_Hammer_Physics
bodies {Anim [{shapeType=Capsule, length=30.4997, radius=18.0858, center=<-0.0013 -1.7021 9.3125>, rotation=<rot -0.1292 -90 0.0033>}]}
constraints []
